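# Niagara system for the RebelTempo emote: a glowing color wheel with rainbow rays and white stars.
# ColorWheel draws two sprite groups plus a light, while Rays continuously spawns seven angled, colored groups scaled by time of day.
# A one-shot star burst accompanies near-identical GPU and CPU emitters of looping twinkles.
# Camera-angle dot products drive wheel and ray fading and camera offset, plus wheel size; front-end switches adjust placement and color.
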NiagaraSystem /BRCosmetics/Animation/Game/MainPlayer/Emotes/RebelTempo/FX/NS_Emote_RebelTempo
dependencies {
 usedDataInterfaces [
  &NiagaraDataInterfaceCurve
  &NiagaraDataInterfaceSpriteRendererInfo
  &NiagaraDataInterfaceCamera
  &NiagaraDataInterfaceVectorCurve
  &NiagaraDataInterfacePlatformSet
 ]
 usedDynamicInputs [
  &FNTimeOfDayScalar
  /Game/Effects/Niagara/DynamicInputs/Logic/FN_PickLinearColorDependOnPlatform
  &self:ScratchDynamicInput_01
 ]
 usedModules [
  &SystemState
  &self:SystemSpawnScript.SystemScriptSource.SystemScriptGraph.NiagaraNodeAssignment_0.SetVariables_#1
  &self:SetPeeloutFXArrayIndex
  &self:SystemSpawnScript.SystemScriptSource.SystemScriptGraph.NiagaraNodeAssignment_1.SetVariables_#2
 ]
}
emitters {
 ColorWheel {
  emitterUpdateScript [
   &EmitterState(Life Cycle Mode=::NewEnumerator1, Loop Behavior=::NewEnumerator1, Loop Duration Mode=::NewEnumerator1, UseLoopDelay=false)
   &SpawnBurst_Instantaneous(Use Spawn Probability=false, UseLoopCountLimit=false) {module=SpawnBurst_Instantaneous001}
   &SpawnBurst_Instantaneous(Spawn Group=1, Use Spawn Probability=false, UseLoopCountLimit=false)
  ]
  particleSpawnScript [
   /SetVariables(Particles.VisibilityTag=&SpawnGroup()) {isSetParameters=true}
   &InitializeParticle(
    Color=&MergeHelper_ColorSwitch(
     0=&MakeCustomLinearColorFromBool(Bool=$User.bIsFrontEndPreview, True Color=<rgba 0.8 0.8 0.8 0.7>, False Color=<rgba 2.5 2.5 2.5 0.85>)
     1=<rgba 1 1 1 0.35>
     2=?
     3=?
     Index=&SpawnGroup()
     Num=::NewEnumerator0
    )
    Lifetime=5
    Position Offset=&FN_FrontEndVector3Modifier(Gameplay=<0 0 0>, Front End=<-10 0 0>)
    Sprite Size=&MergeHelper_Vector2Switch(0=<290 290>, 1=<500 50>, 2=?, 3=?, 4=?, Index=&SpawnGroup(), Num=::NewEnumerator0)
    Sprite Size Mode=::NewEnumerator4
    UsePositionOffset=true
    Write Material Random=true
   ) {enabled=false, module=InitializeParticle003}
   &InitializeParticle(
    Color=&MergeHelper_ColorSwitch(
     0=&MakeCustomLinearColorFromBool(Bool=$User.bIsFrontEndPreview, True Color=<rgba 0.8 0.8 0.8 0.7>, False Color=$System.ValueColorIndoor)
     1=<rgba 1 1 1 0.35>
     2=?
     3=?
     Index=&SpawnGroup()
     Num=::NewEnumerator0
    )
    Lifetime=5
    Position Offset=&FN_FrontEndVector3Modifier(Gameplay=<0 0 0>, Front End=<-10 0 0>)
    Sprite Size=&MergeHelper_Vector2Switch(0=<290 290>, 1=<500 50>, 2=?, 3=?, 4=?, Index=&SpawnGroup(), Num=::NewEnumerator0)
    Sprite Size Mode=::NewEnumerator4
    UsePositionOffset=true
    Write Material Random=true
   )
   &SpriteFacingAndAlignment(
    Alignment Coordinate Space=::Local
    Sprite Alignment=&MergeHelper_VectorSwitch(0=<0 0 1>, 1=<0 1 0>, 2=?, 3=?, Index=&SpawnGroup(), Num=::NewEnumerator0)
    Sprite Facing=&MaskVector(Mask Source=0, Mask Parameter=?, Mask=0, Masked Vector=<0 1 0>, Passthrough Vector=<1 1 1>)
    WriteAlignment=true
    WriteFacing=true
   )
  ]
  particleUpdateScript [
   &ParticleState(Kill Particles When Lifetime Has Elapsed=false)
   &ScaleSpriteSize(
    Uniform Curve Index=$System.Age
    Uniform Curve Sprite Scale=DataInterface() {propertyValues {Curve {keys [<at 0 0 RCIM_Cubic>, {arriveTangent=-0.015, interpMode=RCIM_Cubic, leaveTangent=-0.015, tangentMode=RCTM_User, time=0.35, value=1}]}}}
   )
   &ScaleSpriteSize(
    Scale Sprite Size Mode=::NewEnumerator0
    Uniform Scale Factor=&RemapRange(
     Input Value To Remap=&MakeFloatFromVector(
      Vector=&VectorFromCurve(
       VectorCurve=DataInterface() {propertyValues {XCurve {keys [<at 0 1>,<at 1 0>]}, YCurve {keys [<at 0 1>,<at 1 0>]}, ZCurve {keys [<at 0 1>,<at 1 0>]}}}
       CurveIndex=&DotProduct(
        Vector A=&ReturnCameraVector(Camera Vectors=::NewEnumerator4, Coordinate Space=::Simulation, Origin Position=$Particles.Position)
        Vector B=<0 1 0>
       )
      )
      Channel=::NewEnumerator0
     )
     Input Min=0.8
     Input Max=1
     Output Min=1
     Output Max=0.5
    )
   ) {module=ScaleSpriteSize002}
   &ScaleSpriteSize(
    Uniform Curve Index=$System.Age
    Uniform Curve Scale=&RemapRange(
     Input Value To Remap=&MakeFloatFromVector(
      Vector=&VectorFromCurve(
       VectorCurve=DataInterface() {propertyValues {XCurve {keys [<at 0 1>,<at 1 0>]}, YCurve {keys [<at 0 1>,<at 1 0>]}, ZCurve {keys [<at 0 1>,<at 1 0>]}}}
       CurveIndex=&DotProduct(
        Vector A=&ReturnCameraVector(Camera Vectors=::NewEnumerator4, Coordinate Space=::Simulation, Origin Position=$Particles.Position)
        Vector B=<0 1 0>
       )
      )
      Channel=::NewEnumerator0
     )
     Input Min=0.8
     Input Max=1
     Output Min=1
     Output Max=0.7
    )
    Uniform Curve Sprite Scale=DataInterface() {propertyValues {Curve {keys [<at 0 0 RCIM_Cubic>, {arriveTangent=-0.015, interpMode=RCIM_Cubic, leaveTangent=-0.015, tangentMode=RCTM_User, time=0.35, value=1}]}}}
   ) {enabled=false, module=ScaleSpriteSize001}
   &ScaleColor(
    Scale Alpha=&MergeHelper_FloatSwitch(
     0=&FloatFromCurve(
      FloatCurve=DataInterface() {propertyValues {Curve {
         defaultValue=1
         keys [
          {interpMode=RCIM_Cubic, leaveTangentWeight=0.3333, tangentMode=RCTM_User, time=0, value=0.1}
          {arriveTangentWeight=0.3333, interpMode=RCIM_Cubic, tangentMode=RCTM_User, time=0.1, value=1}
         ]
        }}}
      CurveIndex=&Abs_Float(Float:NiagaraFloat=&DotProduct(
        Vector A=&ReturnCameraVector(
         Camera Vectors=::NewEnumerator4
         Coordinate Space=::Local
         Origin Position=$Particles.Position
         CameraQuery=DataInterface() {propertyValues {bRequireCurrentFrameData=false}}
        )
        Vector B=$Particles.SpriteFacing
       ))
     )
     1=&Add_Float(
      A=&RemapRange(
       Input Value To Remap=&Abs_Float(Float:NiagaraFloat=&DotProduct(Vector A=&ReturnCameraVector(Camera Vectors=::NewEnumerator3, Coordinate Space=::Simulation), Vector B=<1 1 0>))
       Input Min=0.7
       Input Max=0.3
       Output Min=0
       Output Max=0.5
      )
      B=&RemapRange(
       Input Value To Remap=&Abs_Float(Float:NiagaraFloat=&DotProduct(Vector A=&ReturnCameraVector(Camera Vectors=::NewEnumerator1, Coordinate Space=::Simulation), Vector B=<0 1 0>))
       Input Min=0
       Input Max=0.3
       Output Min=0
       Output Max=0.5
      )
     )
     2=?
     3=?
     Index=&SpawnGroup()
     Num=::NewEnumerator0
    )
    ScaleA=true
    ScaleRGB=true
   )
   &ScaleColor(
    Scale Alpha=&Multiply_Float(
     A=0.0015
     B=&MergeHelper_FloatSwitch(
      0=&FloatFromCurve(
       FloatCurve=DataInterface() {propertyValues {Curve {
          defaultValue=1
          keys [
           {interpMode=RCIM_Cubic, leaveTangentWeight=0.3333, tangentMode=RCTM_User, time=0, value=0.1}
           {arriveTangentWeight=0.3333, interpMode=RCIM_Cubic, tangentMode=RCTM_User, time=0.1, value=1}
          ]
         }}}
       CurveIndex=&Abs_Float(Float:NiagaraFloat=&DotProduct(
         Vector A=&ReturnCameraVector(
          Camera Vectors=::NewEnumerator4
          Coordinate Space=::Local
          Origin Position=$Particles.Position
          CameraQuery=DataInterface() {propertyValues {bRequireCurrentFrameData=false}}
         )
         Vector B=$Particles.SpriteFacing
        ))
      )
      1=&Add_Float(
       A=&RemapRange(
        Input Value To Remap=&Abs_Float(Float:NiagaraFloat=&DotProduct(Vector A=&ReturnCameraVector(Camera Vectors=::NewEnumerator3, Coordinate Space=::Simulation), Vector B=<1 1 0>))
        Input Min=0.7
        Input Max=0.3
        Output Min=0
        Output Max=0.5
       )
       B=&RemapRange(
        Input Value To Remap=&Abs_Float(Float:NiagaraFloat=&DotProduct(Vector A=&ReturnCameraVector(Camera Vectors=::NewEnumerator1, Coordinate Space=::Simulation), Vector B=<0 1 0>))
        Input Min=0
        Input Max=0.3
        Output Min=0
        Output Max=0.5
       )
      )
      2=?
      3=?
      Index=&SpawnGroup()
      Num=::NewEnumerator0
     )
    )
    Scale RGB=&Multiply_VectorByFloat(Vector=<1 1 1>, Float=100)
    ScaleA=true
    ScaleRGB=true
   ) {enabled=false, module=ScaleColor003}
   &Light_Attributes(
    Color=1
    Diffuse Scale=1
    Enabled=1
    Light Diffuse Scale=1.7
    Light Position Offset:Vector3f=<-13 20 -50>
    Light Radius=130
    Light Specular Scale=0
    Position=1
    Radius=1
    Scale Alpha:NiagaraFloat=&FloatFromCurve(FloatCurve=DataInterface() {propertyValues {Curve {
        defaultValue=1
        keys [{interpMode=RCIM_Cubic, tangentMode=RCTM_User, time=0, value=0}, {interpMode=RCIM_Cubic, tangentMode=RCTM_User, time=0.2, value=5}]
       }}})
    Scale RGB:Vector3f=&MakeVectorFromLinearColorRGB(Color=<rgba 0.1536 0.0221 1 1>)
    ScaleA=true
    ScaleRGB=true
    Specular Scale=1
    UsePositionOffset=true
   )
   &ScaleColor(
    Scale Alpha=&MaskFloat(
     Mask Source=0
     Mask Parameter=?
     Mask=0
     Masked Float=&RemapRange(
      Input Value To Remap=&MakeFloatFromVector(
       Vector=&VectorFromCurve(
        VectorCurve=DataInterface() {propertyValues {
          XCurve {
           defaultValue=1
           keys [{interpMode=RCIM_Cubic, tangentMode=RCTM_User, time=0, value=1}, {interpMode=RCIM_Cubic, tangentMode=RCTM_User, time=1, value=0}]
          }
          YCurve {
           defaultValue=1
           keys [{interpMode=RCIM_Cubic, tangentMode=RCTM_User, time=0, value=1}, {interpMode=RCIM_Cubic, tangentMode=RCTM_User, time=1, value=0}]
          }
          ZCurve {
           defaultValue=1
           keys [{interpMode=RCIM_Cubic, tangentMode=RCTM_User, time=0, value=1}, {interpMode=RCIM_Cubic, tangentMode=RCTM_User, time=1, value=0}]
          }
         }}
        CurveIndex=&DotProduct(
         Vector A=&ReturnCameraVector(Camera Vectors=::NewEnumerator4, Coordinate Space=::Simulation, Origin Position=$Particles.Position)
         Vector B=<0 1 0>
        )
       )
       Channel=::NewEnumerator0
      )
      Input Min=0
      Input Max=0.9
      Output Min=1
      Output Max=0.0015
     )
     Passthrough Float=1
    )
    Scale RGB=&MergeHelper_VectorSwitch(
     0=&VectorFromFloat(Value=&RemapRange(
       Input Value To Remap=&MakeFloatFromVector(
        Vector=&VectorFromCurve(
         VectorCurve=DataInterface() {propertyValues {
           XCurve {
            defaultValue=1
            keys [{interpMode=RCIM_Cubic, tangentMode=RCTM_User, time=0, value=1}, {interpMode=RCIM_Cubic, tangentMode=RCTM_User, time=1, value=0}]
           }
           YCurve {
            defaultValue=1
            keys [{interpMode=RCIM_Cubic, tangentMode=RCTM_User, time=0, value=1}, {interpMode=RCIM_Cubic, tangentMode=RCTM_User, time=1, value=0}]
           }
           ZCurve {
            defaultValue=1
            keys [{interpMode=RCIM_Cubic, tangentMode=RCTM_User, time=0, value=1}, {interpMode=RCIM_Cubic, tangentMode=RCTM_User, time=1, value=0}]
           }
          }}
         CurveIndex=&DotProduct(
          Vector A=&ReturnCameraVector(Camera Vectors=::NewEnumerator4, Coordinate Space=::Simulation, Origin Position=$Particles.Position)
          Vector B=<0 1 0>
         )
        )
        Channel=::NewEnumerator0
       )
       Input Min=0.3
       Input Max=0.7
       Output Min=1
       Output Max=100
      ))
     1=&VectorFromFloat(Value=&RemapRange(
       Input Value To Remap=&MakeFloatFromVector(
        Vector=&VectorFromCurve(
         VectorCurve=DataInterface() {propertyValues {
           XCurve {
            defaultValue=1
            keys [
             {interpMode=RCIM_Cubic, tangentMode=RCTM_User, time=0, value=1}
             {arriveTangent=-1.0571, interpMode=RCIM_Cubic, leaveTangent=-1.0571, tangentMode=RCTM_User, time=1, value=0}
            ]
           }
           YCurve {
            defaultValue=1
            keys [
             {interpMode=RCIM_Cubic, tangentMode=RCTM_User, time=0, value=1}
             {arriveTangent=-1.0571, interpMode=RCIM_Cubic, leaveTangent=-1.0571, tangentMode=RCTM_User, time=1, value=0}
            ]
           }
           ZCurve {
            defaultValue=1
            keys [
             {interpMode=RCIM_Cubic, tangentMode=RCTM_User, time=0, value=1}
             {arriveTangent=-1.0571, interpMode=RCIM_Cubic, leaveTangent=-1.0571, tangentMode=RCTM_User, time=1, value=0}
            ]
           }
          }}
         CurveIndex=&DotProduct(
          Vector A=&ReturnCameraVector(Camera Vectors=::NewEnumerator4, Coordinate Space=::Simulation, Origin Position=$Particles.Position)
          Vector B=<0 1 0>
         )
        )
        Channel=::NewEnumerator0
       )
       Input Min=0.8
       Input Max=0.9
       Output Min=1
       Output Max=1
      ))
     2=?
     3=?
     Index=&SpawnGroup()
     Num=::NewEnumerator0
    )
    ScaleA=true
    ScaleRGB=true
   ) {enabled=false, module=ScaleColor001}
   &ScaleColor(
    Scale Alpha=&MergeHelper_FloatSwitch(
     0=&RemapRange(
      Input Value To Remap=&MakeFloatFromVector(
       Vector=&VectorFromCurve(
        VectorCurve=DataInterface() {propertyValues {
          XCurve {
           defaultValue=1
           keys [{interpMode=RCIM_Cubic, tangentMode=RCTM_User, time=0, value=1}, {interpMode=RCIM_Cubic, tangentMode=RCTM_User, time=1, value=0}]
          }
          YCurve {
           defaultValue=1
           keys [{interpMode=RCIM_Cubic, tangentMode=RCTM_User, time=0, value=1}, {interpMode=RCIM_Cubic, tangentMode=RCTM_User, time=1, value=0}]
          }
          ZCurve {
           defaultValue=1
           keys [{interpMode=RCIM_Cubic, tangentMode=RCTM_User, time=0, value=1}, {interpMode=RCIM_Cubic, tangentMode=RCTM_User, time=1, value=0}]
          }
         }}
        CurveIndex=&DotProduct(
         Vector A=&ReturnCameraVector(Camera Vectors=::NewEnumerator4, Coordinate Space=::Simulation, Origin Position=$Particles.Position)
         Vector B=<0 1 0>
        )
       )
       Channel=::NewEnumerator0
      )
      Input Min=0.9
      Input Max=1
      Output Min=1
      Output Max=0.05
     )
     1=&RemapRange(
      Input Value To Remap=&MakeFloatFromVector(
       Vector=&VectorFromCurve(
        VectorCurve=DataInterface() {propertyValues {
          XCurve {
           defaultValue=1
           keys [{interpMode=RCIM_Cubic, tangentMode=RCTM_User, time=0, value=1}, {interpMode=RCIM_Cubic, tangentMode=RCTM_User, time=1, value=0}]
          }
          YCurve {
           defaultValue=1
           keys [{interpMode=RCIM_Cubic, tangentMode=RCTM_User, time=0, value=1}, {interpMode=RCIM_Cubic, tangentMode=RCTM_User, time=1, value=0}]
          }
          ZCurve {
           defaultValue=1
           keys [{interpMode=RCIM_Cubic, tangentMode=RCTM_User, time=0, value=1}, {interpMode=RCIM_Cubic, tangentMode=RCTM_User, time=1, value=0}]
          }
         }}
        CurveIndex=&DotProduct(
         Vector A=&ReturnCameraVector(Camera Vectors=::NewEnumerator4, Coordinate Space=::Simulation, Origin Position=$Particles.Position)
         Vector B=<0 1 0>
        )
       )
       Channel=::NewEnumerator0
      )
      Input Min=0.4
      Input Max=0.7
      Output Min=1
      Output Max=0.3
     )
     2=?
     3=?
     Index=&SpawnGroup()
     Num=::NewEnumerator0
    )
    Scale RGB=&VectorFromFloat(Value=&RemapRange(
      Input Value To Remap=&MakeFloatFromVector(
       Vector=&VectorFromCurve(
        VectorCurve=DataInterface() {propertyValues {
          XCurve {defaultValue=1, keys [{tangentMode=RCTM_User, time=0, value=1}, {tangentMode=RCTM_User, time=1, value=0}]}
          YCurve {defaultValue=1, keys [{tangentMode=RCTM_User, time=0, value=1}, {tangentMode=RCTM_User, time=1, value=0}]}
          ZCurve {defaultValue=1, keys [{tangentMode=RCTM_User, time=0, value=1}, {tangentMode=RCTM_User, time=1, value=0}]}
         }}
        CurveIndex=&DotProduct(
         Vector A=&ReturnCameraVector(Camera Vectors=::NewEnumerator4, Coordinate Space=::Simulation, Origin Position=$Particles.Position)
         Vector B=<0 1 0>
        )
       )
       Channel=::NewEnumerator0
      )
      Input Min=0.8
      Input Max=3
      Output Min=1
      Output Max=1
     ))
    ScaleA=true
    ScaleRGB=true
   ) {module=ScaleColor002}
   &CameraOffset(Camera Offset Amount=&RemapRange(
     Input Value To Remap=&MakeFloatFromVector(
      Vector=&VectorFromCurve(
       VectorCurve=DataInterface() {propertyValues {XCurve {keys [<at 0 1>,<at 1 0>]}, YCurve {keys [<at 0 1>,<at 1 0>]}, ZCurve {keys [<at 0 1>,<at 1 0>]}}}
       CurveIndex=&DotProduct(
        Vector A=&ReturnCameraVector(
         Camera Vectors=::NewEnumerator4
         Coordinate Space=::Simulation
         Origin Position=$Particles.Position
         CameraQuery=DataInterface() {propertyValues {bRequireCurrentFrameData=false}}
        )
        Vector B=<0 1 0>
       )
      )
      Channel=::NewEnumerator0
     )
     Input Min=0.2
     Input Max=0.9
     Output Min=-15
     Output Max=40
    )) {module=CameraOffset001}
  ]
  rendererClasses [&NiagaraSpriteRendererProperties, &NiagaraLightRendererProperties]
  renderers [
   {rendererClass=&NiagaraSpriteRendererProperties}
   {rendererClass=&NiagaraLightRendererProperties, rendererIndex=1}
   {rendererClass=&NiagaraSpriteRendererProperties, rendererIndex=2}
  ]
 }
 Rays {
  emitterUpdateScript [
   &EmitterState(Life Cycle Mode=::NewEnumerator1, Loop Behavior=::NewEnumerator1, Loop Duration Mode=::NewEnumerator1, UseLoopDelay=false)
   &SpawnBurst_Instantaneous(Use Spawn Probability=false, UseLoopCountLimit=false)
   &SpawnRate(SpawnRate=2.5, Use Spawn Probability=false)
   &SpawnBurst_Instantaneous(Spawn Group=1, Use Spawn Probability=false, UseLoopCountLimit=false) {module=SpawnBurst_Instantaneous001}
   &SpawnRate(SpawnGroup=1, SpawnRate=2.5, Use Spawn Probability=false) {module=SpawnRate001}
   &SpawnBurst_Instantaneous(Spawn Group=2, Use Spawn Probability=false, UseLoopCountLimit=false) {module=SpawnBurst_Instantaneous002}
   &SpawnRate(SpawnGroup=2, SpawnRate=2.5, Use Spawn Probability=false) {module=SpawnRate002}
   &SpawnBurst_Instantaneous(Spawn Group=3, Use Spawn Probability=false, UseLoopCountLimit=false) {module=SpawnBurst_Instantaneous003}
   &SpawnRate(SpawnGroup=3, SpawnRate=2.5, Use Spawn Probability=false) {module=SpawnRate003}
   &SpawnBurst_Instantaneous(Spawn Group=4, Use Spawn Probability=false, UseLoopCountLimit=false) {module=SpawnBurst_Instantaneous004}
   &SpawnRate(SpawnGroup=4, SpawnRate=2.5, Use Spawn Probability=false) {module=SpawnRate004}
   &SpawnBurst_Instantaneous(Spawn Group=5, Use Spawn Probability=false, UseLoopCountLimit=false) {module=SpawnBurst_Instantaneous005}
   &SpawnRate(SpawnGroup=5, SpawnRate=2.5, Use Spawn Probability=false) {module=SpawnRate005}
   &SpawnBurst_Instantaneous(Spawn Group=6, Use Spawn Probability=false, UseLoopCountLimit=false) {module=SpawnBurst_Instantaneous006}
   &SpawnRate(SpawnGroup=6, SpawnRate=2.5, Use Spawn Probability=false) {module=SpawnRate006}
  ]
  particleSpawnScript [
   /SetVariables(Particles.VisibilityTag=&SpawnGroup()) {isSetParameters=true}
   &InitializeParticle(
    Color=&MergeHelper_ColorSwitch(
     0=<rgba 2 1.9897 0 0.2>
     1=<rgba 3.5 0.762 0 0.2>
     2=<rgba 3.5 0.0004 0.0029 0.2>
     3=&MakeCustomLinearColorFromBool(Bool=$User.bIsFrontEndPreview, True Color=<rgba 0.334 0 4 0.5>, False Color=<rgba 0.2922 0 3.5 0.5>)
     4=&MakeCustomLinearColorFromBool(Bool=$User.bIsFrontEndPreview, True Color=<rgba 0 0 1.5 1>, False Color=<rgba 0 0 2 0.5>)
     5=<rgba 0.0078 2.0196 3.5 0.15>
     6=<rgba 0.1822 3.5 0 0.2>
     Index=&SpawnGroup()
     Num=::NewEnumerator5
    )
    Lifetime Max=3.5
    Lifetime Min=2.5
    Lifetime Mode=::NewEnumerator1
    Position Offset=&FN_FrontEndVector3Modifier(Gameplay=<0 0 10>, Front End=<-10 0 0>)
    Position Offset Coordinate Space=::World
    Sprite Rotation Angle=&MergeHelper_FloatSwitch(0=-50, 1=-105, 2=-150, 3=150, 4=110, 5=65, 6=15, Index=&SpawnGroup(), Num=::NewEnumerator5)
    Sprite Rotation Mode=::NewEnumerator0
    Sprite Size=&FN_FrontEndVector2Modifier(
     Front End:Vector2f=&RandomRangeVector2D(Minimum=<170 120>, Maximum=<200 170>)
     Gameplay:Vector2f=&RandomRangeVector2D(Minimum=<150 120>, Maximum=<190 170>)
    )
    Sprite Size Mode=::NewEnumerator4
    UsePositionOffset=true
    Write Material Random=true
   ) {enabled=false, module=InitializeParticle002}
   &InitializeParticle(
    Color=&MergeHelper_ColorSwitch(
     0=<rgba 2 1.9897 0 0.2>
     1=<rgba 3.5 0.762 0 0.2>
     2=<rgba 3.5 0.0004 0.0029 0.2>
     3=&MakeCustomLinearColorFromBool(Bool=$User.bIsFrontEndPreview, True Color=<rgba 0.334 0 4 0.5>, False Color=<rgba 0.2922 0 3.5 0.5>)
     4=&MakeCustomLinearColorFromBool(Bool=$User.bIsFrontEndPreview, True Color=<rgba 0 0 1.5 1>, False Color=<rgba 0 0 2 0.5>)
     5=<rgba 0.0078 2.0196 3.5 0.15>
     6=<rgba 0.1822 3.5 0 0.2>
     Index=&SpawnGroup()
     Num=::NewEnumerator5
    )
    Lifetime Max=3.5
    Lifetime Min=3
    Lifetime Mode=::NewEnumerator1
    Position Offset=&FN_FrontEndVector3Modifier(Gameplay=<-8 0 10>, Front End=<-29 0 10>)
    Sprite Rotation Angle=&MergeHelper_FloatSwitch(0=-50, 1=-105, 2=-150, 3=150, 4=110, 5=65, 6=15, Index=&SpawnGroup(), Num=::NewEnumerator5)
    Sprite Rotation Mode=::NewEnumerator0
    Sprite Size=&FN_FrontEndVector2Modifier(
     Front End:Vector2f=&RandomRangeVector2D(Minimum=<170 120>, Maximum=<200 170>)
     Gameplay:Vector2f=&RandomRangeVector2D(Minimum=<150 120>, Maximum=<200 180>)
    )
    Sprite Size Mode=::NewEnumerator4
    UsePositionOffset=true
    Write Material Random=true
   ) {module=InitializeParticle001}
   &SubUVAnimation(
    Random Start Frame=true
    "Sprite Renderer "=DataInterface() {propertyValues {SpriteRenderer=&self:Rays_0.NiagaraSpriteRendererProperties_1}}
    Start Frame Offset:NiagaraInt32=&RandomRangeInt(Minimum=0, Maximum=4)
    UseEndFrame=false
    UseStartFrame=false
   ) {module=SubUVAnimation002}
   &SpriteFacingAndAlignment(Sprite Alignment=<0 0 1>, Sprite Facing=<0 1 0>, WriteAlignment=true, WriteFacing=true)
  ]
  particleUpdateScript [
   &ParticleState()
   &SolveForcesAndVelocity(Clamp Velocity=false, Limit Acceleration=false)
   &ScaleColor(
    Scale Alpha=&Multiply_Float(
     A=&FloatFromCurve(
      FloatCurve=DataInterface() {propertyValues {Curve {
         defaultValue=1
         keys [
          {interpMode=RCIM_Cubic, tangentMode=RCTM_User, time=0, value=0}
          {arriveTangent=0.9928, interpMode=RCIM_Cubic, leaveTangent=0.9928, time=0.25, value=1}
          {interpMode=RCIM_Cubic, tangentMode=RCTM_User, time=0.5036, value=0.5}
          {arriveTangent=-1.0073, interpMode=RCIM_Cubic, leaveTangent=-1.0073, time=0.75, value=1}
          {tangentMode=RCTM_User, time=1, value=0}
         ]
        }}}
      Scale Curve=0.5
     )
     B=&FloatFromCurve(
      FloatCurve=DataInterface() {propertyValues {Curve {
         defaultValue=1
         keys [
          {interpMode=RCIM_Cubic, leaveTangentWeight=0.3333, tangentMode=RCTM_User, time=0, value=0}
          {arriveTangentWeight=0.3333, interpMode=RCIM_Cubic, tangentMode=RCTM_User, time=1, value=1}
         ]
        }}}
      CurveIndex=&Abs_Float(Float:NiagaraFloat=&DotProduct(
        Vector A=&ReturnCameraVector(
         Camera Vectors=::NewEnumerator4
         Coordinate Space=::Local
         Origin Position=$Particles.Position
         CameraQuery=DataInterface() {propertyValues {bRequireCurrentFrameData=false}}
        )
        Vector B=$Particles.SpriteFacing
       ))
     )
    )
    Scale RGB=&Multiply_VectorByFloat(Vector=<1 1 1>, Float=&RandomRangeFloat(Minimum=0.5, Maximum=1))
    ScaleA=true
    ScaleRGB=true
   )
   &ScaleColor(
    Scale RGB=&VectorFromFloat(Value=&Multiply_Float(A=1, B=$System.ToD))
    ScaleA=true
    ScaleRGB=true
   ) {module=ScaleColor003}
   &ScaleColor(
    Scale Alpha=&RemapRange(
     Input Value To Remap=&MakeFloatFromVector(
      Vector=&VectorFromCurve(
       VectorCurve=DataInterface() {propertyValues {XCurve {keys [<at 0 1>,<at 1 0>]}, YCurve {keys [<at 0 1>,<at 1 0>]}, ZCurve {keys [<at 0 1>,<at 1 0>]}}}
       CurveIndex=&DotProduct(
        Vector A=&ReturnCameraVector(Camera Vectors=::NewEnumerator4, Coordinate Space=::Simulation, Origin Position=$Particles.Position)
        Vector B=<0 1 0>
       )
      )
      Channel=::NewEnumerator0
     )
     Input Min=0.8
     Input Max=1
     Output Min=1
     Output Max=0.8
    )
    ScaleA=true
    ScaleRGB=true
   ) {module=ScaleColor001}
   &CameraOffset(Camera Offset Amount=&RemapRange(
     Input Value To Remap=&MakeFloatFromVector(
      Vector=&VectorFromCurve(
       VectorCurve=DataInterface() {propertyValues {XCurve {keys [<at 0 1>,<at 1 0>]}, YCurve {keys [<at 0 1>,<at 1 0>]}, ZCurve {keys [<at 0 1>,<at 1 0>]}}}
       CurveIndex=&DotProduct(
        Vector A=&ReturnCameraVector(
         Camera Vectors=::NewEnumerator4
         Coordinate Space=::Simulation
         Origin Position=$Particles.Position
         CameraQuery=DataInterface() {propertyValues {bRequireCurrentFrameData=false}}
        )
        Vector B=<0 1 0>
       )
      )
      Channel=::NewEnumerator0
     )
     Input Min=0.2
     Input Max=0.9
     Output Min=-15
     Output Max=40
    )) {module=CameraOffset001}
  ]
  rendererClasses [&NiagaraSpriteRendererProperties]
  renderers [
   {rendererClass=&NiagaraSpriteRendererProperties}
   {rendererClass=&NiagaraSpriteRendererProperties, rendererIndex=1}
   {rendererClass=&NiagaraSpriteRendererProperties, rendererIndex=2}
   {rendererClass=&NiagaraSpriteRendererProperties, rendererIndex=3}
   {rendererClass=&NiagaraSpriteRendererProperties, rendererIndex=4}
   {rendererClass=&NiagaraSpriteRendererProperties, rendererIndex=5}
   {rendererClass=&NiagaraSpriteRendererProperties, rendererIndex=6}
  ]
 }
 Stars_Burst001 {
  emitterUpdateScript [
   &EmitterState(Life Cycle Mode=::NewEnumerator1, Loop Behavior=::NewEnumerator2, Loop Duration=1, UseLoopDelay=false)
   &SpawnBurst_Instantaneous(Spawn Count=45, Spawn Time=0.04, Use Spawn Probability=false, UseLoopCountLimit=false)
   &SpawnBurst_Instantaneous(Spawn Count=35, Spawn Group=1, Spawn Time=0.06, Use Spawn Probability=false, UseLoopCountLimit=false) {module=SpawnBurst_Instantaneous001}
  ]
  particleSpawnScript [
   /SetVariables(Particles.VisibilityTag=&SpawnGroup()) {isSetParameters=true}
   &InitializeParticle(
    Color=<rgba 2.5 2.5 2.5 1>
    Lifetime=&MergeHelper_FloatSwitch(
     0=&RandomRangeFloat(Minimum=0.4, Maximum=1)
     1=&RandomRangeFloat(Minimum=0.4, Maximum=0.9)
     2=?
     3=?
     Index=&SpawnGroup()
     Num=::NewEnumerator0
    )
    Position Offset=<0 0 140>
    Sprite Size Mode=::NewEnumerator1
    Uniform Sprite Size Max=25
    Uniform Sprite Size Min=10
    UsePositionOffset=false
    Write Material Random=true
   )
   &AddVelocity(
    Constrain To Radius=200
    Module.Use Falloff Distance=false
    Override Seed=false
    Velocity Mode=::NewEnumerator1
    Velocity Speed=&MergeHelper_FloatSwitch(
     0=&RandomRangeFloat(Minimum=1200, Maximum=900)
     1=&RandomRangeFloat(Minimum=480, Maximum=530)
     2=?
     3=?
     Index=&SpawnGroup()
     Num=::NewEnumerator0
    )
   )
   &ShapeLocation(
    MaskSpawnGroup=false
    Offset=<0 0 10>
    Offset Mode=::NewEnumerator0
    Override Seed=false
    Sphere Radius=&MergeHelper_FloatSwitch(0=2, 1=10, 2=?, 3=?, Index=&SpawnGroup(), Num=::NewEnumerator0)
    Sphere Surface Distribution=1
    UseOwnerScale=true
   ) {module=ShapeLocation002}
   &CameraOffset(Camera Offset Amount=-40)
  ]
  particleUpdateScript [
   &ParticleState()
   &SpriteRotationRate(Rotation Rate=&RandomRangeFloat(Minimum=5, Maximum=7))
   &ScaleSpriteSize(Uniform Curve Sprite Scale=DataInterface() {propertyValues {Curve {
       defaultValue=0
       keys [
        {interpMode=RCIM_Cubic, leaveTangentWeight=0.3347, tangentMode=RCTM_User, time=0, value=0}
        {interpMode=RCIM_Cubic, tangentMode=RCTM_User, time=1, value=1}
       ]
      }}})
   &ScaleColor(
    Scale Alpha=&MergeHelper_FloatSwitch(
     0=&FloatFromCurve(FloatCurve=DataInterface() {propertyValues {Curve {
         defaultValue=1
         keys [
          {tangentMode=RCTM_User, time=0, value=1}
          {arriveTangent=-0.0405, interpMode=RCIM_Cubic, leaveTangent=-0.0405, tangentMode=RCTM_User, time=0.65, value=1}
          {arriveTangentWeight=1.0761, interpMode=RCIM_Cubic, tangentMode=RCTM_User, time=1, value=0}
         ]
        }}})
     1=&FloatFromCurve(FloatCurve=DataInterface() {propertyValues {Curve {
         defaultValue=1
         keys [{interpMode=RCIM_Cubic, tangentMode=RCTM_User, time=0.3, value=1}, {interpMode=RCIM_Cubic, tangentMode=RCTM_User, time=1, value=0}]
        }}})
     2=?
     3=?
     Index=&SpawnGroup()
     Num=::NewEnumerator0
    )
    ScaleA=true
    ScaleRGB=true
   )
   &Drag(
    Drag=&MergeHelper_FloatSwitch(
     0=&RandomRangeFloat(Minimum=15, Maximum=15)
     1=&RandomRangeFloat(Minimum=15, Maximum=15)
     2=?
     3=?
     Index=&SpawnGroup()
     Num=::NewEnumerator0
    )
    Use Linear Drag=true
    Use Rotational Drag=true
   ) {module=Drag001}
   &CurlNoiseForce(Bias Noise=false, Noise Frequency=10, Noise Strength=3, Pan Noise=false)
   &SolveForcesAndVelocity(Clamp Velocity=false, Limit Acceleration=false)
  ]
  rendererClasses [&NiagaraSpriteRendererProperties]
  renderers [
   {rendererClass=&NiagaraSpriteRendererProperties}
   {rendererClass=&NiagaraSpriteRendererProperties, rendererIndex=1}
  ]
 }
 Stars_Rate_GPU003 {
  simTarget=GPUComputeSim
  emitterUpdateScript [
   &EmitterState(Life Cycle Mode=::NewEnumerator1, Loop Behavior=::NewEnumerator1, Loop Delay=0.3, Loop Duration=1, UseLoopDelay=true)
   &SpawnBurst_Instantaneous(Spawn Count=200, Use Spawn Probability=false, UseLoopCountLimit=false)
  ]
  particleSpawnScript [
   &InitializeParticle(
    Color=<rgba 2.5 2.5 2.5 1>
    Lifetime=&RandomRangeFloat(Minimum=1.5, Maximum=0.5)
    Position Offset=<30 -10 0>
    Position Offset Coordinate Space=::Local
    Sprite Size Mode=::NewEnumerator1
    Uniform Sprite Size Max=30
    Uniform Sprite Size Min=10
    UsePositionOffset=true
    Write Material Random=true
   )
   &ShapeLocation(
    Cylinder Height=15
    MaskSpawnGroup=false
    Offset=<-40 12 0>
    Offset Coordinate Space=::Local
    Offset Mode=::NewEnumerator0
    Override Seed=false
    Rotation Axis=<0 1 0>
    Rotation Mode=::NewEnumerator1
    Shape Primitive=::NewEnumerator1
    Surface Only=false
    UseOwnerScale=true
   ) {module=ShapeLocation001}
   &CameraOffset(Camera Offset Amount=-50)
  ]
  particleUpdateScript [
   &ParticleState(Kill Particles When Lifetime Has Elapsed=false, Loop Particles Lifetime=true)
   &SpriteRotationRate(Rotation Rate=50) {enabled=false}
   &SpriteRotationRate(Delta Time:NiagaraFloat=$Emitter.NormalizedLoopAge, Rotation Rate=5) {module=SpriteRotationRate001}
   &ScaleSpriteSize(Uniform Curve Sprite Scale=DataInterface() {propertyValues {Curve {
       defaultValue=0
       keys [
        {interpMode=RCIM_Cubic, tangentMode=RCTM_User, time=0, value=0}
        {interpMode=RCIM_Cubic, tangentMode=RCTM_User, time=0.5, value=1}
        {tangentMode=RCTM_User, time=1, value=0}
       ]
      }}})
   &ScaleColor(
    Scale Alpha=&FloatFromCurve(FloatCurve=DataInterface() {propertyValues {Curve {
        defaultValue=1
        keys [
         {tangentMode=RCTM_User, time=0, value=1}
         {arriveTangent=-0.0405, interpMode=RCIM_Cubic, leaveTangent=-0.0405, tangentMode=RCTM_User, time=0.92, value=1}
         {arriveTangent=-8.0089, interpMode=RCIM_Cubic, leaveTangent=-8.0089, tangentMode=RCTM_User, time=1, value=0}
        ]
       }}})
    ScaleA=true
    ScaleRGB=true
   )
   &SolveForcesAndVelocity(Clamp Velocity=false, Limit Acceleration=false)
  ]
  rendererClasses [&NiagaraSpriteRendererProperties]
  renderers [{rendererClass=&NiagaraSpriteRendererProperties}]
 }
 Stars_Rate_GPU004 {
  emitterUpdateScript [
   &EmitterState(Life Cycle Mode=::NewEnumerator1, Loop Behavior=::NewEnumerator1, Loop Delay=0.3, Loop Duration=1, UseLoopDelay=true)
   &SpawnBurst_Instantaneous(Spawn Count=200, Use Spawn Probability=false, UseLoopCountLimit=false)
  ]
  particleSpawnScript [
   &InitializeParticle(
    Color=<rgba 2.5 2.5 2.5 1>
    Lifetime=&RandomRangeFloat(Minimum=1.5, Maximum=0.5)
    Position Offset=<30 -10 0>
    Position Offset Coordinate Space=::Local
    Sprite Size Mode=::NewEnumerator1
    Uniform Sprite Size Max=30
    Uniform Sprite Size Min=10
    UsePositionOffset=true
    Write Material Random=true
   )
   &ShapeLocation(
    Cylinder Height=15
    MaskSpawnGroup=false
    Offset=<-40 12 0>
    Offset Coordinate Space=::Local
    Offset Mode=::NewEnumerator0
    Override Seed=false
    Rotation Axis=<0 1 0>
    Rotation Mode=::NewEnumerator1
    Shape Primitive=::NewEnumerator1
    Surface Only=false
    UseOwnerScale=true
   ) {module=ShapeLocation001}
   &CameraOffset(Camera Offset Amount=-50)
  ]
  particleUpdateScript [
   &ParticleState(Kill Particles When Lifetime Has Elapsed=false, Loop Particles Lifetime=true)
   &SpriteRotationRate(Delta Time:NiagaraFloat=$Emitter.NormalizedLoopAge, Rotation Rate=5) {module=SpriteRotationRate001}
   &ScaleSpriteSize(Uniform Curve Sprite Scale=DataInterface() {propertyValues {Curve {
       defaultValue=0
       keys [
        {interpMode=RCIM_Cubic, tangentMode=RCTM_User, time=0, value=0}
        {interpMode=RCIM_Cubic, tangentMode=RCTM_User, time=0.5, value=1}
        {tangentMode=RCTM_User, time=1, value=0}
       ]
      }}})
   &ScaleColor(
    Scale Alpha=&FloatFromCurve(FloatCurve=DataInterface() {propertyValues {Curve {
        defaultValue=1
        keys [
         {tangentMode=RCTM_User, time=0, value=1}
         {arriveTangent=-0.0405, interpMode=RCIM_Cubic, leaveTangent=-0.0405, tangentMode=RCTM_User, time=0.92, value=1}
         {arriveTangent=-8.0089, interpMode=RCIM_Cubic, leaveTangent=-8.0089, tangentMode=RCTM_User, time=1, value=0}
        ]
       }}})
    ScaleA=true
    ScaleRGB=true
   )
   &SolveForcesAndVelocity(Clamp Velocity=false, Limit Acceleration=false)
  ]
  rendererClasses [&NiagaraSpriteRendererProperties]
  renderers [{rendererClass=&NiagaraSpriteRendererProperties}]
 }
}
summary {
 systemName=NS_Emote_RebelTempo
 emitters [
  {emitterName=Stars_Burst001, rendererClasses [&NiagaraSpriteRendererProperties]}
  {emitterName=Stars_Rate_GPU003, simTarget=GPUComputeSim, rendererClasses [&NiagaraSpriteRendererProperties]}
  {emitterName=Stars_Rate_GPU004, rendererClasses [&NiagaraSpriteRendererProperties]}
  {emitterName=Rays, rendererClasses [&NiagaraSpriteRendererProperties]}
  {emitterName=ColorWheel, rendererClasses [&NiagaraSpriteRendererProperties, &NiagaraLightRendererProperties]}
 ]
 userVariables [
  User.bIsFrontEnd {type=NiagaraBool}
  User.bIsFrontEndPreview {type=NiagaraBool}
  User.IsSpawnDust {type=NiagaraBool, defaultValue=true}
  User.SlidingColorArray {
   type=NiagaraDataInterfaceArrayColor
   description=""
   defaultValue {
    struct=NiagaraExt_VariableValue_DataInterface
    value {
     dataInterface=&self:NiagaraDataInterfaceArrayColor_0
     dataInterfaceClass=&NiagaraDataInterfaceArrayColor
    }
   }
  }
  User.SurfaceTypeIndex {type=NiagaraInt32}
 ]
}
